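; Behavior tree service for swarm NPCs that evaluates whether a target actor and the swarm itself are inside the leash radius.
; On each AI tick it casts the controlled pawn to NPC_Pawn_Swarm and reads the actor from a configurable blackboard key.
; IsInsideLeashRadius results go to two blackboard bools, one for that actor and one for the pawn.
; An invalid actor writes false to its bool, while the self check still runs.
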(event AI|EventReceiveTickAI
  (params (OwnerController :type AIController%20Object%20Reference) (ControlledPawn :type Pawn%20Object%20Reference) (DeltaSeconds :type Float%20%28single-precision%29))
  (bind _actorkeytotest (Variables|Default|GetActorKeyToTest))
  (bind _returnvalue (AI|BehaviorTree|GetBlackboardValueasActor _actorkeytotest))
  (bind _asnpc_pawn_swarm (Utilities|Casting|CastToNPC_Pawn_Swarm ControlledPawn)
    (:then
      (Utilities|IsValid _returnvalue
        (:Is%20Valid
          (bind _isinleashradius (Class|NPCPawnSwarm|IsInsideLeashRadius _asnpc_pawn_swarm _returnvalue "0, 0, 0"))
          (AI|BehaviorTree|SetBlackboardValueasBool (Variables|Default|GetBoolToSet) _isinleashradius))
        (:Is%20Not%20Valid
          (AI|BehaviorTree|SetBlackboardValueasBool (Variables|Default|GetBoolToSet))))
      (bind _isinleashradius_2 (Class|NPCPawnSwarm|IsInsideLeashRadius _asnpc_pawn_swarm _asnpc_pawn_swarm "0, 0, 0"))
      (AI|BehaviorTree|SetBlackboardValueasBool (Variables|Default|GetSelfLeashBoolToSet) _isinleashradius_2))
    (:CastFailed
      _)))

(event EventReceiveDeactivation
  (params (OwnerActor :type Actor%20Object%20Reference)))
