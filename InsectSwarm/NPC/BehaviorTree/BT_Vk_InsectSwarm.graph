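# Root behavior tree for the Vk insect swarm NPC, switching between default swarm behavior and the Vk AI behavior.
# A Selector runs the BT_InsectSwarm subtree when the VK_DefaultBehaviorExecutionStatus blackboard value is at least ExecutionPending.
# If that branch is blocked or fails, it falls through to the BT_Vk_InsectSwarm_AI subtree, which has no gating decorator.
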BehaviorTree /InsectSwarm/NPC/BehaviorTree/BT_Vk_InsectSwarm
blackboard=&/BB_Vk_InsectSwarm
root {
 name=Selector
 class=BTComposite_Selector
 children [
  Run Default behavior {
   class=BTTask_RunBehavior
   decorators [Default Behavior Status {
     class=BTDecorator_Blackboard
     properties {
      arithmeticOperation=GreaterOrEqual
      intValue=3
      nodeName=Default Behavior Status
      stringValue=EExecutionStatus::ExecutionPending
      blackboardKey {selectedKeyName=VK_DefaultBehaviorExecutionStatus}
     }
    }]
   properties {behaviorAsset=&/BT_InsectSwarm, nodeName=Run Default behavior}
  }
  Run Behavior {
   class=BTTask_RunBehavior
   decorators []
   properties {behaviorAsset=&/BT_Vk_InsectSwarm_AI, nodeName=Run Behavior}
  }
 ]
 decorators []
 properties {
  children [
   {
    childComposite=null
    childTask=&self:BTTask_RunBehavior_3
    decorators [&self:BTDecorator_Blackboard_47]
   }
   {childComposite=null, childTask=&self:BTTask_RunBehavior_9, decorators []}
  ]
  services []
 }
 services []
}
